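# Physics asset for the Loofah Female one-handed pickaxe cosmetic, giving collision bodies and joint limits to its dyn_ fruit bones: watermelon, orange, lemon and olive.
# Each fruit has a fitted capsule sized to it, while the handle bone has a small capsule as the shared anchor.
# Every fruit constrains directly to the handle rather than chaining, with tight twist and asymmetric swing limits that are widest on watermelon and narrowest on olive along swing2.
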PhysicsAsset /BRCosmetics/Weapons/FORT_Melee/Pickaxe_LoofahFemale_1h/Meshes/Loofah_Female_1h_Axe_Physics
bodies {
 Melee_TwoHanded_Handle [{shapeType=Capsule, length=1, radius=1}]
 dyn_lemon [{shapeType=Capsule, length=0.9679, radius=2.8798, center=<0.0093 0.2674 -0.3478>, rotation=<rot 1.5996 -101.0329 -81.8579>}]
 dyn_olive [{shapeType=Capsule, length=2.75, radius=2.5134, center=<0 0 -0.4488>, rotation=<rot 1.4518 -104.0528 -84.2203>}]
 dyn_orange [{shapeType=Capsule, length=2.5, radius=6.5776, center=<-0.5215 -0.2355 -0.5425>, rotation=<rot -9.494 -71.557 -63.6828>}]
 dyn_watermelon [{shapeType=Capsule, length=13.1083, radius=7.7509, center=<-2.571 -0.0101 -0.3921>, rotation=<rot 0 -89.9667 -80.9105>}]
}
constraints [
 {bone1Name=dyn_watermelon, bone2Name=Melee_TwoHanded_Handle, swing1LimitDegrees=11, swing2LimitDegrees=55, twistLimitDegrees=7}
 {bone1Name=dyn_orange, bone2Name=Melee_TwoHanded_Handle, swing1LimitDegrees=10, swing2LimitDegrees=38.5714, twistLimitDegrees=5}
 {bone1Name=dyn_lemon, bone2Name=Melee_TwoHanded_Handle, swing1LimitDegrees=15, swing2LimitDegrees=45, twistLimitDegrees=5}
 {bone1Name=dyn_olive, bone2Name=Melee_TwoHanded_Handle, swing1LimitDegrees=16, swing2LimitDegrees=17.7143, twistLimitDegrees=5}
]
